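# Test Level Sequence for the LimeRock Test2 level instance base that spins a cone static mesh actor.
# It possesses the Cone actor and drives it with a single 3D transform track.
# Roll (Rotation.X) is keyed through one full negative turn, while pitch and yaw stay at zero.
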LevelSequence /LimeRock/Levels/Test/LimeRock_Test2_LevelInstance_Base_Sequence
bindings [Cone {
  id=#1
  possessed_class=FortStaticMeshActor
  tracks [{
    type=MovieScene3DTransformTrack
    sections [{
      type=MovieScene3DTransformSection
      channels [
       Rotation.X {type=MovieSceneScriptingDoubleChannel, default=0, thinned_from=5, keys [[0,0],[120,-360]]}
       Rotation.Y {type=MovieSceneScriptingDoubleChannel, default=0, const [0,5]}
       Rotation.Z {type=MovieSceneScriptingDoubleChannel, default=0, const [0,5]}
      ]
     }]
   }]
 }]
display_rate {denominator=1, numerator=30}
tick_resolution {denominator=1, numerator=24000}
tracks []
