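; LimeRock test base Blueprint for exercising Level Sequence playback in a Level Instance.
; Separate custom events validate and play the non-replicated or replicated sequence actor reference, printing debug strings for validity and playback.
; A stop event halts only the non-replicated sequence, without a validity check.
; LR_Event uses the DataLayerManager to activate the LimeRock level instance data layer at runtime.
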(event Custom|LR_LI_Base_Test_NonRep
  (Utilities|IsValid (Variables|LevelInstanceTest|GetLevelSequence%20%28No%20Rep%29)
    (:Is%20Valid
      (Development|PrintString "[LimeRock] Level Instance Test - Level Sequence VALID")
      (Development|PrintString "[LimeRock] Level Instance Test -PLAY SEQUENCE")
      (Sequencer|Player|Play (Variables|/Script/LevelSequence.LevelSequenceActor|GetSequencePlayer (Variables|LevelInstanceTest|GetLevelSequence%20%28No%20Rep%29)) :!class /Script/MovieScene.MovieSceneSequencePlayer))
    (:Is%20Not%20Valid
      (Development|PrintString "[LimeRock] Level Instance Test - Level Sequence NOT valid"))))

(event Custom|LR_LI_Base_Test_Rep
  (Utilities|IsValid (Variables|LevelInstanceTest|GetLevelSequence%20%28Replicated%29)
    (:Is%20Valid
      (Development|PrintString "[LimeRock] Level Instance Test - Level Sequence VALID")
      (Development|PrintString "[LimeRock] Level Instance Test -PLAY SEQUENCE")
      (Sequencer|Player|Play (Variables|/Script/LevelSequence.LevelSequenceActor|GetSequencePlayer (Variables|LevelInstanceTest|GetLevelSequence%20%28Replicated%29)) :!class /Script/MovieScene.MovieSceneSequencePlayer))
    (:Is%20Not%20Valid
      (Development|PrintString "[LimeRock] Level Instance Test - Level Sequence NOT valid"))))

(event Custom|LR_LI_Test_Stop
  (Sequencer|Player|Stop (Variables|/Script/LevelSequence.LevelSequenceActor|GetSequencePlayer (Variables|LevelInstanceTest|GetLevelSequence%20%28No%20Rep%29)) :!class /Script/MovieScene.MovieSceneSequencePlayer))

(event Custom|LR_Event
  (bind _returnvalue (WorldPartition|GetDataLayerManager))
  (DataLayers|SetDataLayerRuntimeState _returnvalue "/LimeRock/DataLayers/DL_LimeRock_LevelInstance.DL_LimeRock_LevelInstance" "Activated" false :!class /Script/Engine.DataLayerManager) ; event actors
)

(fn ConstructionScript)
